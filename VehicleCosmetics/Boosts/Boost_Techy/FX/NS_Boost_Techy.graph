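# Niagara system for the Techy vehicle boost cosmetic, trailing sprites, meshes and ribbons behind the car's filtered boost sockets.
# Two sprite emitters distribute noisy, fading particles across ports, with spawn groups setting render tags; one also emits stationary, scaled meshes.
# Ribbon001 spawns a ribbon per port, drawn by four renderers with separate width and color settings.
# VC_PaintedVariantsOverride supplies painted color variants throughout, using over-life curves on the ribbons.
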NiagaraSystem /VehicleCosmetics/Boosts/Boost_Techy/FX/NS_Boost_Techy
dependencies {
 usedDataInterfaces [
  &NiagaraDataInterfaceColorCurve
  &NiagaraDataInterfaceCurve
  &NiagaraDataInterfaceArrayInt32
  &NiagaraDataInterfaceSkeletalMesh
 ]
 usedDynamicInputs []
 usedModules [
  &self:SystemSpawnScript.SystemScriptSource.SystemScriptGraph.NiagaraNodeAssignment_0.SetVariables_#1
  &SystemState
 ]
}
emitters {
 NE_VC_Boost_Sprite_Parent002 {
  emitterUpdateScript [
   &EmitterState(Inactive Response=::NewEnumerator1, Life Cycle Mode=::NewEnumerator1, Loop Duration=1, UseLoopDelay=false)
   &VC_FilteredSocketCount()
   &SpawnRate(
    SpawnGroup:NiagaraInt32=&SelectIntFromArray(
     Array Sampling Mode=::NewEnumerator1
     Direct Array Index=?
     Evaluation Type=::NewEnumerator1
     Int Selection Array=DataInterface() {propertyValues {IntData [0,2,3]}}
     Random Seed=0
    )
    SpawnRate=10
    Use Spawn Probability=true
   )
   &SpawnRate(SpawnGroup=1, SpawnRate=6, Use Spawn Probability=false) {module=SpawnRate001}
  ]
  particleSpawnScript [
   /SetVariables(Particles.RenderTag:NiagaraInt32=&SpawnGroup()) {isSetParameters=true}
   &VC_PaintedVariantsOverride(
    Black=<rgba 0.001 0.001 0.001 10>
    BurntSienna=<rgba 0.08 0.0053 0.004 5>
    Cobalt=<rgba 0 0 0.45 5>
    Crimson=<rgba 0.2 0 0.0054 5>
    DefaultColor=<rgba 0.25 0.8 1 1>
    ForestGreen=<rgba 0 0.5 0.0253 2>
    Grey=<rgba 0.022 0.022 0.022 5>
    Lime=<rgba 0.8 1.2 0 1>
    Orange=<rgba 0.17 0.051 0 5>
    Pink=<rgba 1 0.07 0.6618 1>
    Purple=<rgba 0.651 0.06 1.5 1>
    R1 Black:LinearColor=<rgba 0.6354 0.6354 0.6354 1>
    R1 BurntSienna:LinearColor=<rgba 0.651 0.201 0.0803 1>
    R1 Cobalt:LinearColor=<rgba 0.2083 0.223 1 1>
    R1 Crimson:LinearColor=<rgba 0.9479 0.2123 0.2322 1>
    R1 ForestGreen:LinearColor=<rgba 0.1146 1 0.1593 1>
    R1 Lime:LinearColor=<rgba 0.6146 1 0.2292 1>
    R1 Orange:LinearColor=<rgba 0.6724 0.3189 0.1296 1>
    R1 Purple:LinearColor=<rgba 0.3373 0.0991 0.6795 1>
    R1 Saffron:LinearColor=<rgba 1 0.7861 0.0833 1>
    R1 SkyBlue:LinearColor=<rgba 0.0938 0.6968 1 1>
    R2 Black:LinearColor=<rgba 0.3005 0.3005 0.3005 0.349>
    R2 BurntSienna:LinearColor=<rgba 0.3005 0.0762 0.0152 0.349>
    R2 Cobalt:LinearColor=<rgba 0.08 0.0822 0.2 0.25>
    R2 Crimson:LinearColor=<rgba 0.402 0.0802 0.0887 0.251>
    R2 DefaultColor:LinearColor=<rgba 0.025 0.1 0.5 0.33>
    R2 ForestGreen:LinearColor=<rgba 0.0578 0.5029 0.0802 0.302>
    R2 Grey:LinearColor=<rgba 0.0908 0.0908 0.0908 0.349>
    R2 Lime:LinearColor=<rgba 0.2747 0.5029 0.0497 0.302>
    R2 Orange:LinearColor=<rgba 0.3005 0.1413 0.0578 0.302>
    R2 Pink:LinearColor=<rgba 0.2502 0.0296 0.1714 0.302>
    R2 Purple:LinearColor=<rgba 0.1746 0.0203 0.402 0.302>
    R2 Saffron:LinearColor=<rgba 0.552 0.4342 0.0452 0.302>
    R2 SkyBlue:LinearColor=<rgba 0.0423 0.314 0.4508 0.302>
    R2 TitaniumWhite:LinearColor=<rgba 0.3515 0.3515 0.3515 0.302>
    Saffron=<rgba 0.25 0.1917 0 5>
    SkyBlue=<rgba 0.06 0.5191 0.75 1>
    Use R1 Painted Overrides=true
    Use R2 Painted Overrides=true
   )
   &InitializeParticle(
    Color=$Particles.Color
    Lifetime=&MergeHelper_FloatSwitch(
     0=&RandomRangeFloat(Minimum=0.4, Maximum=0.8)
     1=1
     2=&RandomRangeFloat(Minimum=0.4, Maximum=0.8)
     3=&RandomRangeFloat(Minimum=0.4, Maximum=0.8)
     4=1
     Index=&SpawnGroup()
     Num=::NewEnumerator3
    )
    Mesh Scale=&MergeHelper_VectorSwitch(0=<0 0 0>, 1=<3 5 5>, 2=<0 0 0>, 3=<0 0 0>, 4=<2.2 5 5>, Index=&SpawnGroup(), Num=::NewEnumerator3)
    Mesh Scale Mode=::NewEnumerator4
    Sprite Rotation Mode=::NewEnumerator2
    Sprite Size Mode=::NewEnumerator3
    Uniform Sprite Size=&RandomRangeFloat(Minimum=10, Maximum=15)
    UsePositionOffset=false
    Write Material Random=true
   )
   &PartitionParticles(Number Of Partitions=$Emitter.PortCount)
   &SkeletalMeshLocation(
    Bone / Socket Index=$Particles.PartitionParticles.Partition
    Bone / Socket Sampling Mode=::NewEnumerator5
    MaskSpawnGroup=false
    Mesh Sampling Type=::NewEnumerator6
    Sampled Position Offset=&AddVectorToPosition(
     Position=$User.DTCarOffset
     Vector=&MergeHelper_VectorSwitch(0=<-40 0 0>, 1=<-80 0 0>, 2=<-40 0 0>, 3=<-40 0 0>, 4=<-80 0 0>, Index=&SpawnGroup(), Num=::NewEnumerator3)
    )
    Skeletal Mesh=$System.SkeletalMesh
   )
   &AddVelocity(
    Velocity=&MergeHelper_VectorSwitch(0=<-500 0 0>, 1=<0 0 0>, 2=<-500 0 0>, 3=<-500 0 0>, 4=<0 0 0>, Index=&SpawnGroup(), Num=::NewEnumerator3)
    Velocity Speed Scale=0.5
   )
   &ShapeLocation(
    MaskSpawnGroup=false
    Override Seed=false
    Sphere Radius=&MergeHelper_FloatSwitch(0=45, 1=0, 2=45, 3=45, 4=0, Index=&SpawnGroup(), Num=::NewEnumerator3)
    UseOwnerScale=true
   ) {module=ShapeLocation001}
  ]
  particleUpdateScript [
   &ParticleState()
   &ScaleColor(
    Scale Alpha=&FloatFromCurve(FloatCurve=DataInterface() {propertyValues {Curve {keys [<at -0.2 1 RCIM_Cubic>,<at 0.8 0 RCIM_Cubic>]}}}, Scale Curve=1.5)
    ScaleA=true
    ScaleRGB=true
   )
   &Drag(Drag=2, Use Linear Drag=true, Use Rotational Drag=true)
   &CurlNoiseForce(Bias Noise=false, Pan Noise=false)
   &SolveForcesAndVelocity(Clamp Velocity=false, Limit Acceleration=false)
   /SetVariables(
    Particles.Index1:Vector4f=&MakeVector4(X=5, Y=10, Z=0, W=5)
    Particles.MeshColor1:LinearColor=&VC_DynamicScaleColor(
     Color to Scale:LinearColor=&MakeCustomLinearColorFromBool(Bool=$User.bIsPainted, True Color=$Particles.Initial.R1 Color, False Color=<rgba 0.05 0.2 1 1>)
     RGB Scale:Vector3f=&VectorFromFloat(Value=&MakeCustomFloatFromBool(Bool=$User.bIsPainted, True Float=1, False Float=1))
     Alpha Scale:NiagaraFloat=&FloatFromCurve(FloatCurve=DataInterface() {propertyValues {Curve {keys [<at 0 1 RCIM_Cubic>,<at 1 0 RCIM_Cubic>]}}})
    )
   ) {isSetParameters=true}
  ]
  rendererClasses [&NiagaraSpriteRendererProperties, &NiagaraMeshRendererProperties]
  renderers [
   {rendererClass=&NiagaraSpriteRendererProperties}
   {rendererClass=&NiagaraSpriteRendererProperties, rendererIndex=1}
   {rendererClass=&NiagaraMeshRendererProperties, rendererIndex=2}
   {rendererClass=&NiagaraSpriteRendererProperties, rendererIndex=3}
  ]
 }
 NE_VC_Boost_Sprite_Parent003 {
  emitterUpdateScript [
   &EmitterState()
   &VC_FilteredSocketCount()
   &SpawnRate(
    SpawnGroup:NiagaraInt32=&RandomRangeInt(Minimum=0, Maximum=2, Evaluation Type=::NewEnumerator1)
    SpawnRate=60
    Use Spawn Probability=true
   )
  ]
  particleSpawnScript [
   /SetVariables(Particles.RenderTag:NiagaraInt32=&SpawnGroup()) {isSetParameters=true}
   &VC_PaintedVariantsOverride(
    Black=<rgba 0.001 0.001 0.001 10>
    BurntSienna=<rgba 0.08 0.0053 0.004 5>
    Cobalt=<rgba 0 0 0.45 5>
    Crimson=<rgba 0.2 0 0.0054 5>
    DefaultColor=<rgba 0.25 0.8 1 1>
    ForestGreen=<rgba 0 0.5 0.0253 2>
    Grey=<rgba 0.022 0.022 0.022 5>
    Lime=<rgba 0.8 1.2 0 1>
    Orange=<rgba 0.17 0.051 0 5>
    Pink=<rgba 1 0.07 0.6618 1>
    Purple=<rgba 0.651 0.06 1.5 1>
    Saffron=<rgba 0.25 0.1917 0 5>
    SkyBlue=<rgba 0.06 0.5191 0.75 1>
   )
   &InitializeParticle(
    Color=$Particles.Color
    Lifetime=&RandomRangeFloat(Minimum=1, Maximum=1.5)
    Sprite Rotation Mode=::NewEnumerator2
    Sprite Size Mode=::NewEnumerator3
    Uniform Sprite Size=&RandomRangeFloat(Minimum=10, Maximum=15)
    UsePositionOffset=false
    Write Material Random=true
   )
   &PartitionParticles(Number Of Partitions=$Emitter.PortCount)
   &SkeletalMeshLocation(
    Bone / Socket Index=$Particles.PartitionParticles.Partition
    Bone / Socket Sampling Mode=::NewEnumerator5
    MaskSpawnGroup=false
    Mesh Sampling Type=::NewEnumerator6
    Sampled Position Offset=&AddVectorToPosition(Position=$User.DTCarOffset, Vector=<-30 0 0>)
    Skeletal Mesh=$System.SkeletalMesh
   )
   &AddVelocity(Velocity=<-400 0 0>)
   &ShapeLocation(MaskSpawnGroup=false, Override Seed=false, Sphere Radius=40, UseOwnerScale=true) {module=ShapeLocation001}
  ]
  particleUpdateScript [
   &ParticleState()
   &InheritSourceMovement(
    Applied Position Scale=&VectorFromFloat(Value=&FloatFromCurve(FloatCurve=DataInterface() {propertyValues {Curve {keys [<at 0 1 RCIM_Cubic>,<at 1 0 RCIM_Cubic>]}}}))
    ApplyForce=false
    ApplyPosition=true
    ApplyPositionFalloff=false
   )
   &CurlNoiseForce(Bias Noise=false, Noise Frequency=300, Noise Strength=100, Pan Noise=false)
   &ScaleColor(
    Scale Alpha=&FloatFromCurve(FloatCurve=DataInterface() {propertyValues {Curve {keys [<at 0 1 RCIM_Cubic>,<at 0.8 0 RCIM_Cubic>]}}})
    Scale RGB=&VectorFromFloat(Value=&MakeCustomFloatFromBool(Bool=$User.bIsPainted, True Float=1, False Float=2))
    ScaleA=true
    ScaleRGB=true
   )
   &SolveForcesAndVelocity(Clamp Velocity=false, Limit Acceleration=false)
  ]
  rendererClasses [&NiagaraSpriteRendererProperties]
  renderers [
   {rendererClass=&NiagaraSpriteRendererProperties}
   {rendererClass=&NiagaraSpriteRendererProperties, rendererIndex=1}
   {rendererClass=&NiagaraSpriteRendererProperties, rendererIndex=2}
  ]
 }
 Ribbon001 {
  emitterUpdateScript [
   &EmitterState()
   &VC_FilteredSocketCount()
   &VC_RateQualityHandler()
   &VC_SpawnRibbon_v1(Number of Ribbons=$Emitter.PortCount, SpawnRate=$Emitter.Spawn Rate)
  ]
  particleSpawnScript [
   &VC_InitializeRibbon_v1(
    Color=$Particles.Color
    Custom Offset=true
    Enable Skeletal Position=true
    Lifetime=0.4
    Local Vector Offset=&AddVectorToPosition(Position=$User.DTCarOffset, Vector=<0 0 5>)
    Offset Amount=1
    Ribbon Width=60
   )
   &AddVelocity(Velocity=<-1500 0 0>, Velocity Speed Scale=2) {module=AddVelocity001}
  ]
  particleUpdateScript [
   &ParticleState()
   &SolveForcesAndVelocity(Clamp Velocity=false, Limit Acceleration=false)
   &VC_UpdateRibbon_v1(Enable Skeletal Position=true)
   &VC_PaintedVariantsOverride(
    Black=&ColorFromCurve(ColorCurve=DataInterface() {propertyValues {
       AlphaCurve {keys [<at 0.6399 4 RCIM_Cubic>,<at 1 0 RCIM_Cubic>]}
       BlueCurve {keys [<at 0 0.05 RCIM_Cubic>]}
       GreenCurve {keys [<at 0 0.05 RCIM_Cubic>]}
       RedCurve {keys [<at 0 0.05 RCIM_Cubic>]}
      }})
    BurntSienna=&ColorFromCurve(ColorCurve=DataInterface() {propertyValues {
       AlphaCurve {keys [<at 0.6476 3 RCIM_Cubic>,<at 1 0 RCIM_Cubic>]}
       BlueCurve {keys [<at 0 0.06 RCIM_Cubic>]}
       GreenCurve {keys [<at 0 0.063 RCIM_Cubic>]}
       RedCurve {keys [<at 0 0.15 RCIM_Cubic>]}
      }})
    Cobalt=&ColorFromCurve(ColorCurve=DataInterface() {propertyValues {
       AlphaCurve {keys [<at 0.652 3.5 RCIM_Cubic>,<at 1 0 RCIM_Cubic>]}
       BlueCurve {keys [<at 0 0.4 RCIM_Cubic>]}
       GreenCurve {keys [<at 0 0 RCIM_Cubic>]}
       RedCurve {keys [<at 0 0.0128 RCIM_Cubic>]}
      }})
    Crimson=&ColorFromCurve(ColorCurve=DataInterface() {propertyValues {
       AlphaCurve {keys [<at 0.6553 3 RCIM_Cubic>,<at 1 0 RCIM_Cubic>]}
       BlueCurve {keys [<at 0 0.0376 RCIM_Cubic>]}
       GreenCurve {keys [<at 0 0 RCIM_Cubic>]}
       RedCurve {keys [<at 0 0.22 RCIM_Cubic>]}
      }})
    DefaultColor=&ColorFromCurve(ColorCurve=DataInterface() {propertyValues {
       AlphaCurve {keys [<at 0.5987 2 RCIM_Cubic>,<at 1 0 RCIM_Cubic>]}
       BlueCurve {keys [<at 0 0.35 RCIM_Cubic>]}
       GreenCurve {keys [<at 0 0.0972 RCIM_Cubic>]}
       RedCurve {keys [<at 0 0.2451 RCIM_Cubic>]}
      }})
    ForestGreen=&ColorFromCurve(ColorCurve=DataInterface() {propertyValues {
       AlphaCurve {keys [<at 0.6608 3.5 RCIM_Cubic>,<at 1 0 RCIM_Cubic>]}
       BlueCurve {keys [<at 0 0.0932 RCIM_Cubic>]}
       GreenCurve {keys [<at 0 0.16 RCIM_Cubic>]}
       RedCurve {keys [<at 0 0 RCIM_Cubic>]}
      }})
    Gold:LinearColor=<rgba 0.2462 0.0976 0.3515 1>
    Grey=&ColorFromCurve(ColorCurve=DataInterface() {propertyValues {
       AlphaCurve {keys [<at 0.6642 3 RCIM_Cubic>,<at 1 0 RCIM_Cubic>]}
       BlueCurve {keys [<at 0 0.18 RCIM_Cubic>]}
       GreenCurve {keys [<at 0 0.18 RCIM_Cubic>]}
       RedCurve {keys [<at 0 0.18 RCIM_Cubic>]}
      }})
    Lime=&ColorFromCurve(ColorCurve=DataInterface() {propertyValues {
       AlphaCurve {keys [<at 0.6851 3.5 RCIM_Cubic>,<at 1 0 RCIM_Cubic>]}
       BlueCurve {keys [<at 0 0 RCIM_Cubic>]}
       GreenCurve {keys [<at 0 0.16 RCIM_Cubic>]}
       RedCurve {keys [<at 0 0.134 RCIM_Cubic>]}
      }})
    Orange=&ColorFromCurve(ColorCurve=DataInterface() {propertyValues {
       AlphaCurve {keys [<at 0.6906 2 RCIM_Cubic>,<at 1 0>]}
       BlueCurve {keys [<at 0 0.1054 RCIM_Cubic>]}
       GreenCurve {keys [<at 0 0.1192 RCIM_Cubic>]}
       RedCurve {keys [<at 0 0.1986 RCIM_Cubic>]}
      }})
    Pink=&ColorFromCurve(ColorCurve=DataInterface() {propertyValues {
       AlphaCurve {keys [<at 0.694 2 RCIM_Cubic>,<at 1 0 RCIM_Cubic>]}
       BlueCurve {keys [<at 0 0.144 RCIM_Cubic>]}
       GreenCurve {keys [<at 0 0.04 RCIM_Cubic>]}
       RedCurve {keys [<at 0 0.2 RCIM_Cubic>]}
      }})
    Purple=&ColorFromCurve(ColorCurve=DataInterface() {propertyValues {
       AlphaCurve {keys [<at 0.6796 2.5 RCIM_Cubic>,<at 1 0 RCIM_Cubic>]}
       BlueCurve {keys [<at 0 0.3 RCIM_Cubic>]}
       GreenCurve {keys [<at 0 0 RCIM_Cubic>]}
       RedCurve {keys [<at 0 0.15 RCIM_Cubic>]}
      }})
    R1 Black:LinearColor=&ColorFromCurve(ColorCurve=DataInterface() {propertyValues {
       AlphaCurve {keys [<at 0 0.8 RCIM_Cubic>,<at 0.8021 0 RCIM_Cubic>]}
       BlueCurve {keys [<at 0 0.7 RCIM_Cubic>]}
       GreenCurve {keys [<at 0 0.7 RCIM_Cubic>]}
       RedCurve {keys [<at 0 0.7 RCIM_Cubic>]}
      }})
    R1 BurntSienna:LinearColor=&ColorFromCurve(ColorCurve=DataInterface() {propertyValues {
       AlphaCurve {keys [<at 0.2426 0.8 RCIM_Cubic>,<at 0.7811 0 RCIM_Cubic>]}
       BlueCurve {keys [<at 0 0.456 RCIM_Cubic>]}
       GreenCurve {keys [<at 0 0.5605 RCIM_Cubic>]}
       RedCurve {keys [<at 0 0.95 RCIM_Cubic>]}
      }})
    R1 Cobalt:LinearColor=&ColorFromCurve(ColorCurve=DataInterface() {propertyValues {
       AlphaCurve {keys [<at 0 0.8 RCIM_Cubic>,<at 0.7922 0 RCIM_Cubic>]}
       BlueCurve {keys [<at 0 3.5 RCIM_Cubic>]}
       GreenCurve {keys [<at 0 0.7 RCIM_Cubic>]}
       RedCurve {keys [<at 0 0 RCIM_Cubic>]}
      }})
    R1 Crimson:LinearColor=&ColorFromCurve(ColorCurve=DataInterface() {propertyValues {
       AlphaCurve {keys [<at 0 0.8 RCIM_Cubic>,<at 0.812 0 RCIM_Cubic>]}
       BlueCurve {keys [<at 0 0.2042 RCIM_Cubic>]}
       GreenCurve {keys [<at 0 0.125 RCIM_Cubic>]}
       RedCurve {keys [<at 0 2.5 RCIM_Cubic>]}
      }})
    R1 DefaultColor=&ColorFromCurve(ColorCurve=DataInterface() {propertyValues {
       AlphaCurve {keys [<at 0 0.8>,<at 0.7888 0 RCIM_Cubic>]}
       BlueCurve {keys [<at 0.007 10>,<at 0.2813 4.5>]}
       GreenCurve {keys [<at 0.007 3.5083>,<at 0.2813 1.425>]}
       RedCurve {keys [<at 0.007 0.5>,<at 0.2813 0>]}
      }})
    R1 ForestGreen:LinearColor=&ColorFromCurve(ColorCurve=DataInterface() {propertyValues {
       AlphaCurve {keys [<at 0 0.6 RCIM_Cubic>,<at 0.8131 0 RCIM_Cubic>]}
       BlueCurve {keys [<at 0 0.3625 RCIM_Cubic>]}
       GreenCurve {keys [<at 0 2.5 RCIM_Cubic>]}
       RedCurve {keys [<at 0 0.25 RCIM_Cubic>]}
      }})
    R1 Gold:LinearColor=<rgba 0.1027 0.044 0.22 1.5>
    R1 Grey:LinearColor=&ColorFromCurve(ColorCurve=DataInterface() {propertyValues {
       AlphaCurve {keys [<at 0 1 RCIM_Cubic>,<at 0.8253 0 RCIM_Cubic>]}
       BlueCurve {keys [<at 0 1.5 RCIM_Cubic>]}
       GreenCurve {keys [<at 0 1.5 RCIM_Cubic>]}
       RedCurve {keys [<at 0 1.5 RCIM_Cubic>]}
      }})
    R1 Lime:LinearColor=&ColorFromCurve(ColorCurve=DataInterface() {propertyValues {
       AlphaCurve {keys [<at 0 1 RCIM_Cubic>,<at 0.8186 0 RCIM_Cubic>]}
       BlueCurve {keys [<at 0 0 RCIM_Cubic>]}
       GreenCurve {keys [<at 0 2 RCIM_Cubic>]}
       RedCurve {keys [<at 0 1.4333 RCIM_Cubic>]}
      }})
    R1 Orange:LinearColor=&ColorFromCurve(ColorCurve=DataInterface() {propertyValues {
       AlphaCurve {keys [<at 0 1 RCIM_Cubic>,<at 0.8098 0 RCIM_Cubic>]}
       BlueCurve {keys [<at 0 0.5 RCIM_Cubic>]}
       GreenCurve {keys [<at 0 1.6667 RCIM_Cubic>]}
       RedCurve {keys [<at 0 2.5 RCIM_Cubic>]}
      }})
    R1 Pink:LinearColor=&ColorFromCurve(ColorCurve=DataInterface() {propertyValues {
       AlphaCurve {keys [<at 0 0.62 RCIM_Cubic>,<at 0.8253 0 RCIM_Cubic>]}
       BlueCurve {keys [<at 0 2.4791 RCIM_Cubic>]}
       GreenCurve {keys [<at 0 1.216 RCIM_Cubic>]}
       RedCurve {keys [<at 0 3.2 RCIM_Cubic>]}
      }})
    R1 Purple:LinearColor=&ColorFromCurve(ColorCurve=DataInterface() {propertyValues {
       AlphaCurve {keys [<at 0 0.6 RCIM_Cubic>,<at 0.8286 0 RCIM_Cubic>]}
       BlueCurve {keys [<at 0 4 RCIM_Cubic>]}
       GreenCurve {keys [<at 0 0.4 RCIM_Cubic>]}
       RedCurve {keys [<at 0 2.5 RCIM_Cubic>]}
      }})
    R1 Saffron:LinearColor=&ColorFromCurve(ColorCurve=DataInterface() {propertyValues {
       AlphaCurve {keys [<at 0 1 RCIM_Cubic>,<at 0.8286 0 RCIM_Cubic>]}
       BlueCurve {keys [<at 0 0.4 RCIM_Cubic>]}
       GreenCurve {keys [<at 0 1.8933 RCIM_Cubic>]}
       RedCurve {keys [<at 0 2 RCIM_Cubic>]}
      }})
    R1 SkyBlue:LinearColor=&ColorFromCurve(ColorCurve=DataInterface() {propertyValues {
       AlphaCurve {keys [<at 0 1 RCIM_Cubic>,<at 0.8286 0 RCIM_Cubic>]}
       BlueCurve {keys [<at 0 2 RCIM_Cubic>]}
       GreenCurve {keys [<at 0 1.625 RCIM_Cubic>]}
       RedCurve {keys [<at 0 0.5 RCIM_Cubic>]}
      }})
    R1 TitaniumWhite:LinearColor=&ColorFromCurve(ColorCurve=DataInterface() {propertyValues {
       AlphaCurve {keys [<at 0 1 RCIM_Cubic>,<at 0.8175 0 RCIM_Cubic>]}
       BlueCurve {keys [<at 0 2 RCIM_Cubic>]}
       GreenCurve {keys [<at 0 2 RCIM_Cubic>]}
       RedCurve {keys [<at 0 2 RCIM_Cubic>]}
      }})
    R2 Black:LinearColor=&ColorFromCurve(ColorCurve=DataInterface() {propertyValues {
       AlphaCurve {keys [<at 0 0.1>,<at 0.7988 0 RCIM_Cubic>]}
       BlueCurve {keys [<at 0 0.05 RCIM_Cubic>]}
       GreenCurve {keys [<at 0 0.05 RCIM_Cubic>]}
       RedCurve {keys [<at 0 0.05 RCIM_Cubic>]}
      }})
    R2 BurntSienna:LinearColor=&ColorFromCurve(ColorCurve=DataInterface() {propertyValues {
       AlphaCurve {keys [<at 0 0.15 RCIM_Cubic>,<at 0.8804 0 RCIM_Cubic>]}
       BlueCurve {keys [<at 0 0.0066 RCIM_Cubic>]}
       GreenCurve {keys [<at 0 0.0167 RCIM_Cubic>]}
       RedCurve {keys [<at 0 0.055 RCIM_Cubic>]}
      }})
    R2 Cobalt:LinearColor=&ColorFromCurve(ColorCurve=DataInterface() {propertyValues {
       AlphaCurve {keys [<at 0 0.1 RCIM_Cubic>,<at 0.8374 0 RCIM_Cubic>]}
       BlueCurve {keys [<at 0 0.3 RCIM_Cubic>]}
       GreenCurve {keys [<at 0 0.0346 RCIM_Cubic>]}
       RedCurve {keys [<at 0 0.03 RCIM_Cubic>]}
      }})
    R2 Crimson:LinearColor=&ColorFromCurve(ColorCurve=DataInterface() {propertyValues {
       AlphaCurve {keys [<at 0 0.15 RCIM_Cubic>,<at 0.8551 0 RCIM_Cubic>]}
       BlueCurve {keys [<at 0 0.025 RCIM_Cubic>]}
       GreenCurve {keys [<at 0 0.006 RCIM_Cubic>]}
       RedCurve {keys [<at 0 0.12 RCIM_Cubic>]}
      }})
    R2 DefaultColor:LinearColor=&ColorFromCurve(ColorCurve=DataInterface() {propertyValues {
       AlphaCurve {keys [<at 0 0.1 RCIM_Cubic>,<at 0.8098 0 RCIM_Cubic>]}
       BlueCurve {keys [<at 0 0.19 RCIM_Cubic>]}
       GreenCurve {keys [<at 0 0.0697 RCIM_Cubic>]}
       RedCurve {keys [<at 0 0.0095 RCIM_Cubic>]}
      }})
    R2 ForestGreen:LinearColor=&ColorFromCurve(ColorCurve=DataInterface() {propertyValues {
       AlphaCurve {keys [<at 0 0.2 RCIM_Cubic>,<at 0.8131 0 RCIM_Cubic>]}
       BlueCurve {keys [<at 0 0.0233 RCIM_Cubic>]}
       GreenCurve {keys [<at 0 0.08 RCIM_Cubic>]}
       RedCurve {keys [<at 0 0.012 RCIM_Cubic>]}
      }})
    R2 Grey:LinearColor=&ColorFromCurve(ColorCurve=DataInterface() {propertyValues {
       AlphaCurve {keys [<at 0 0.1 RCIM_Cubic>,<at 0.7999 0 RCIM_Cubic>]}
       BlueCurve {keys [<at 0 0.08 RCIM_Cubic>]}
       GreenCurve {keys [<at 0 0.08 RCIM_Cubic>]}
       RedCurve {keys [<at 0 0.08 RCIM_Cubic>]}
      }})
    R2 Lime:LinearColor=&ColorFromCurve(ColorCurve=DataInterface() {propertyValues {
       AlphaCurve {keys [<at 0 0.1 RCIM_Cubic>,<at 0.7523 0 RCIM_Cubic>]}
       BlueCurve {keys [<at 0 0 RCIM_Cubic>]}
       GreenCurve {keys [<at 0 0.08 RCIM_Cubic>]}
       RedCurve {keys [<at 0 0.064 RCIM_Cubic>]}
      }})
    R2 Orange:LinearColor=&ColorFromCurve(ColorCurve=DataInterface() {propertyValues {
       AlphaCurve {keys [<at 0 0.1 RCIM_Cubic>,<at 0.7922 0 RCIM_Cubic>]}
       BlueCurve {keys [<at 0 0.012 RCIM_Cubic>]}
       GreenCurve {keys [<at 0 0.0325 RCIM_Cubic>]}
       RedCurve {keys [<at 0 0.08 RCIM_Cubic>]}
      }})
    R2 Pink:LinearColor=&ColorFromCurve(ColorCurve=DataInterface() {propertyValues {
       AlphaCurve {keys [<at 0 0.1 RCIM_Cubic>,<at 0.8032 0 RCIM_Cubic>]}
       BlueCurve {keys [<at 0 0.0903 RCIM_Cubic>]}
       GreenCurve {keys [<at 0 0.036 RCIM_Cubic>]}
       RedCurve {keys [<at 0 0.12 RCIM_Cubic>]}
      }})
    R2 Purple:LinearColor=&ColorFromCurve(ColorCurve=DataInterface() {propertyValues {
       AlphaCurve {keys [<at 0 0.08 RCIM_Cubic>,<at 0.8065 0 RCIM_Cubic>]}
       BlueCurve {keys [<at 0 0.25 RCIM_Cubic>]}
       GreenCurve {keys [<at 0 0 RCIM_Cubic>]}
       RedCurve {keys [<at 0 0.1026 RCIM_Cubic>]}
      }})
    R2 Saffron:LinearColor=&ColorFromCurve(ColorCurve=DataInterface() {propertyValues {
       AlphaCurve {keys [<at 0 0.08 RCIM_Cubic>,<at 0.801 0 RCIM_Cubic>]}
       BlueCurve {keys [<at 0 0.018 RCIM_Cubic>]}
       GreenCurve {keys [<at 0 0.0962 RCIM_Cubic>]}
       RedCurve {keys [<at 0 0.12 RCIM_Cubic>]}
      }})
    R2 SkyBlue:LinearColor=&ColorFromCurve(ColorCurve=DataInterface() {propertyValues {
       AlphaCurve {keys [<at 0 0.1 RCIM_Cubic>,<at 0.8043 0 RCIM_Cubic>]}
       BlueCurve {keys [<at 0 0.14 RCIM_Cubic>]}
       GreenCurve {keys [<at 0 0.0957 RCIM_Cubic>]}
       RedCurve {keys [<at 0 0.007 RCIM_Cubic>]}
      }})
    R2 TitaniumWhite:LinearColor=&ColorFromCurve(ColorCurve=DataInterface() {propertyValues {
       AlphaCurve {keys [<at 0 0.1 RCIM_Cubic>,<at 0.7789 0 RCIM_Cubic>]}
       BlueCurve {keys [<at 0 0.08 RCIM_Cubic>]}
       GreenCurve {keys [<at 0 0.08 RCIM_Cubic>]}
       RedCurve {keys [<at 0 0.08 RCIM_Cubic>]}
      }})
    R3 Black:LinearColor=&ColorFromCurve(ColorCurve=DataInterface() {propertyValues {
       AlphaCurve {keys [<at 0.2708 0.3 RCIM_Cubic>,<at 0.8222 0 RCIM_Cubic>]}
       BlueCurve {keys [<at 0 0.22 RCIM_Cubic>]}
       GreenCurve {keys [<at 0 0.22 RCIM_Cubic>]}
       RedCurve {keys [<at 0 0.22 RCIM_Cubic>]}
      }})
    R3 BurntSienna:LinearColor=&ColorFromCurve(ColorCurve=DataInterface() {propertyValues {
       AlphaCurve {keys [<at 0.265 0.3 RCIM_Cubic>,<at 0.8279 0 RCIM_Cubic>]}
       BlueCurve {keys [<at 0 0.1462 RCIM_Cubic>]}
       GreenCurve {keys [<at 0 0.2056 RCIM_Cubic>]}
       RedCurve {keys [<at 0 0.43 RCIM_Cubic>]}
      }})
    R3 Cobalt:LinearColor=&ColorFromCurve(ColorCurve=DataInterface() {propertyValues {
       AlphaCurve {keys [<at 0.2642 0.2 RCIM_Cubic>,<at 0.8468 0 RCIM_Cubic>]}
       BlueCurve {keys [<at 0 1.8 RCIM_Cubic>]}
       GreenCurve {keys [<at 0 0.2255 RCIM_Cubic>]}
       RedCurve {keys [<at 0 0.198 RCIM_Cubic>]}
      }})
    R3 Crimson:LinearColor=&ColorFromCurve(ColorCurve=DataInterface() {propertyValues {
       AlphaCurve {keys [<at 0.2609 0.2 RCIM_Cubic>,<at 0.855 0 RCIM_Cubic>]}
       BlueCurve {keys [<at 0 0.1829 RCIM_Cubic>]}
       GreenCurve {keys [<at 0 0.16 RCIM_Cubic>]}
       RedCurve {keys [<at 0 1 RCIM_Cubic>]}
      }})
    R3 DefaultColor:LinearColor=&ColorFromCurve(ColorCurve=DataInterface() {propertyValues {
       AlphaCurve {keys [<at 0.293 0.3 RCIM_Cubic>,<at 0.8893 0 RCIM_Cubic>]}
       BlueCurve {keys [<at 0 0.7 RCIM_Cubic>]}
       GreenCurve {keys [<at 0 0.2934 RCIM_Cubic>]}
       RedCurve {keys [<at 0 0.105 RCIM_Cubic>]}
      }})
    R3 ForestGreen:LinearColor=&ColorFromCurve(ColorCurve=DataInterface() {propertyValues {
       AlphaCurve {keys [<at 0.2691 0.2 RCIM_Cubic>,<at 0.8526 0 RCIM_Cubic>]}
       BlueCurve {keys [<at 0 0.2391 RCIM_Cubic>]}
       GreenCurve {keys [<at 0 0.5 RCIM_Cubic>]}
       RedCurve {keys [<at 0 0.225 RCIM_Cubic>]}
      }})
    R3 Grey:LinearColor=&ColorFromCurve(ColorCurve=DataInterface() {propertyValues {
       AlphaCurve {keys [<at 0.2708 0.5 RCIM_Cubic>,<at 0.8509 0 RCIM_Cubic>]}
       BlueCurve {keys [<at 0 0.26 RCIM_Cubic>]}
       GreenCurve {keys [<at 0 0.26 RCIM_Cubic>]}
       RedCurve {keys [<at 0 0.26 RCIM_Cubic>]}
      }})
    R3 Lime:LinearColor=&ColorFromCurve(ColorCurve=DataInterface() {propertyValues {
       AlphaCurve {keys [<at 0.2741 0.3 RCIM_Cubic>,<at 0.8518 0 RCIM_Cubic>]}
       BlueCurve {keys [<at 0 0.188 RCIM_Cubic>]}
       GreenCurve {keys [<at 0 0.4 RCIM_Cubic>]}
       RedCurve {keys [<at 0 0.3575 RCIM_Cubic>]}
      }})
    R3 Orange:LinearColor=&ColorFromCurve(ColorCurve=DataInterface() {propertyValues {
       AlphaCurve {keys [<at 0.288 0.3 RCIM_Cubic>,<at 0.8435 0 RCIM_Cubic>]}
       BlueCurve {keys [<at 0 0.172 RCIM_Cubic>]}
       GreenCurve {keys [<at 0 0.2498 RCIM_Cubic>]}
       RedCurve {keys [<at 0 0.43 RCIM_Cubic>]}
      }})
    R3 Pink:LinearColor=&ColorFromCurve(ColorCurve=DataInterface() {propertyValues {
       AlphaCurve {keys [<at 0.293 0.3 RCIM_Cubic>,<at 0.8468 0 RCIM_Cubic>]}
       BlueCurve {keys [<at 0 0.3787 RCIM_Cubic>]}
       GreenCurve {keys [<at 0 0.2115 RCIM_Cubic>]}
       RedCurve {keys [<at 0 0.47 RCIM_Cubic>]}
      }})
    R3 Purple:LinearColor=&ColorFromCurve(ColorCurve=DataInterface() {propertyValues {
       AlphaCurve {keys [<at 0.2889 0.3 RCIM_Cubic>,<at 0.8485 0 RCIM_Cubic>]}
       BlueCurve {keys [<at 0 0.65 RCIM_Cubic>]}
       GreenCurve {keys [<at 0 0.0975 RCIM_Cubic>]}
       RedCurve {keys [<at 0 0.3243 RCIM_Cubic>]}
      }})
    R3 Saffron:LinearColor=&ColorFromCurve(ColorCurve=DataInterface() {propertyValues {
       AlphaCurve {keys [<at 0.2921 0.3 RCIM_Cubic>,<at 0.8493 0 RCIM_Cubic>]}
       BlueCurve {keys [<at 0 0.154 RCIM_Cubic>]}
       GreenCurve {keys [<at 0 0.2506 RCIM_Cubic>]}
       RedCurve {keys [<at 0 0.28 RCIM_Cubic>]}
      }})
    R3 SkyBlue:LinearColor=&ColorFromCurve(ColorCurve=DataInterface() {propertyValues {
       AlphaCurve {keys [<at 0.2905 0.3 RCIM_Cubic>,<at 0.8526 0 RCIM_Cubic>]}
       BlueCurve {keys [<at 0 0.5 RCIM_Cubic>]}
       GreenCurve {keys [<at 0 0.3667 RCIM_Cubic>]}
       RedCurve {keys [<at 0 0.1 RCIM_Cubic>]}
      }})
    R3 TitaniumWhite:LinearColor=&ColorFromCurve(ColorCurve=DataInterface() {propertyValues {
       AlphaCurve {keys [<at 0.2856 0.3 RCIM_Cubic>,<at 0.8583 0 RCIM_Cubic>]}
       BlueCurve {keys [<at 0 0.28 RCIM_Cubic>]}
       GreenCurve {keys [<at 0 0.28 RCIM_Cubic>]}
       RedCurve {keys [<at 0 0.28 RCIM_Cubic>]}
      }})
    R4 Black:LinearColor=&ColorFromCurve(ColorCurve=DataInterface() {propertyValues {
       AlphaCurve {keys [<at 0 6 RCIM_Cubic>,<at 0.1837 0 RCIM_Cubic>]}
       BlueCurve {keys [<at 0 0.09 RCIM_Cubic>]}
       GreenCurve {keys [<at 0 0.09 RCIM_Cubic>]}
       RedCurve {keys [<at 0 0.09 RCIM_Cubic>]}
      }})
    R4 BurntSienna:LinearColor=&ColorFromCurve(ColorCurve=DataInterface() {propertyValues {
       AlphaCurve {keys [<at 0 0>,<at 0.0558 6 RCIM_Cubic>,<at 0.1641 0 RCIM_Cubic>]}
       BlueCurve {keys [<at 0 0 RCIM_Cubic>]}
       GreenCurve {keys [<at 0 0.0256 RCIM_Cubic>]}
       RedCurve {keys [<at 0 0.12 RCIM_Cubic>]}
      }})
    R4 Cobalt:LinearColor=&ColorFromCurve(ColorCurve=DataInterface() {propertyValues {
       AlphaCurve {keys [<at 0 0>,<at 0.0558 6 RCIM_Cubic>,<at 0.1641 0 RCIM_Cubic>]}
       BlueCurve {keys [<at 0 0.1 RCIM_Cubic>]}
       GreenCurve {keys [<at 0 0.0408 RCIM_Cubic>]}
       RedCurve {keys [<at 0 0.0398 RCIM_Cubic>]}
      }})
    R4 Crimson:LinearColor=&ColorFromCurve(ColorCurve=DataInterface() {propertyValues {
       AlphaCurve {keys [<at 0 0>,<at 0.0558 6 RCIM_Cubic>,<at 0.1509 0 RCIM_Cubic>]}
       BlueCurve {keys [<at 0 0.0265 RCIM_Cubic>]}
       GreenCurve {keys [<at 0 0.0239 RCIM_Cubic>]}
       RedCurve {keys [<at 0 0.12 RCIM_Cubic>]}
      }})
    R4 DefaultColor:LinearColor=&ColorFromCurve(ColorCurve=DataInterface() {propertyValues {
       AlphaCurve {keys [<at 0 8 RCIM_Cubic>,<at 0.1641 0 RCIM_Cubic>]}
       BlueCurve {keys [<at 0 0.1 RCIM_Cubic>]}
       GreenCurve {keys [<at 0 0.0156 RCIM_Cubic>]}
       RedCurve {keys [<at 0 0 RCIM_Cubic>]}
      }})
    R4 ForestGreen:LinearColor=&ColorFromCurve(ColorCurve=DataInterface() {propertyValues {
       AlphaCurve {keys [<at 0 0>,<at 0.0558 6 RCIM_Cubic>,<at 0.1509 0 RCIM_Cubic>]}
       BlueCurve {keys [<at 0 0.0239 RCIM_Cubic>]}
       GreenCurve {keys [<at 0 0.15 RCIM_Cubic>]}
       RedCurve {keys [<at 0 0.0172 RCIM_Cubic>]}
      }})
    R4 Grey:LinearColor=&ColorFromCurve(ColorCurve=DataInterface() {propertyValues {
       AlphaCurve {keys [<at 0 0>,<at 0.0558 6 RCIM_Cubic>,<at 0.1509 0 RCIM_Cubic>]}
       BlueCurve {keys [<at 0 0.03 RCIM_Cubic>]}
       GreenCurve {keys [<at 0 0.03 RCIM_Cubic>]}
       RedCurve {keys [<at 0 0.03 RCIM_Cubic>]}
      }})
    R4 Lime:LinearColor=&ColorFromCurve(ColorCurve=DataInterface() {propertyValues {
       AlphaCurve {keys [<at 0 0>,<at 0.0558 6 RCIM_Cubic>,<at 0.1509 0 RCIM_Cubic>]}
       BlueCurve {keys [<at 0 0.0148 RCIM_Cubic>]}
       GreenCurve {keys [<at 0 0.15 RCIM_Cubic>]}
       RedCurve {keys [<at 0 0.0819 RCIM_Cubic>]}
      }})
    R4 Orange:LinearColor=&ColorFromCurve(ColorCurve=DataInterface() {propertyValues {
       AlphaCurve {keys [<at 0 0>,<at 0.0616 6 RCIM_Cubic>,<at 0.1386 0 RCIM_Cubic>]}
       BlueCurve {keys [<at 0 0.0173 RCIM_Cubic>]}
       GreenCurve {keys [<at 0 0.0423 RCIM_Cubic>]}
       RedCurve {keys [<at 0 0.09 RCIM_Cubic>]}
      }})
    R4 Pink:LinearColor=&ColorFromCurve(ColorCurve=DataInterface() {propertyValues {
       AlphaCurve {keys [<at 0 0>,<at 0.0616 6 RCIM_Cubic>,<at 0.1501 0 RCIM_Cubic>]}
       BlueCurve {keys [<at 0 0.0685 RCIM_Cubic>]}
       GreenCurve {keys [<at 0 0.0118 RCIM_Cubic>]}
       RedCurve {keys [<at 0 0.1 RCIM_Cubic>]}
      }})
    R4 Purple:LinearColor=&ColorFromCurve(ColorCurve=DataInterface() {propertyValues {
       AlphaCurve {keys [<at 0 0>,<at 0.0616 6 RCIM_Cubic>,<at 0.1501 0 RCIM_Cubic>]}
       BlueCurve {keys [<at 0 0.1 RCIM_Cubic>]}
       GreenCurve {keys [<at 0 0.005 RCIM_Cubic>]}
       RedCurve {keys [<at 0 0.0434 RCIM_Cubic>]}
      }})
    R4 Saffron:LinearColor=&ColorFromCurve(ColorCurve=DataInterface() {propertyValues {
       AlphaCurve {keys [<at 0 0>,<at 0.0616 6 RCIM_Cubic>,<at 0.1501 0 RCIM_Cubic>]}
       BlueCurve {keys [<at 0 0.0123 RCIM_Cubic>]}
       GreenCurve {keys [<at 0 0.118 RCIM_Cubic>]}
       RedCurve {keys [<at 0 0.15 RCIM_Cubic>]}
      }})
    R4 SkyBlue:LinearColor=&ColorFromCurve(ColorCurve=DataInterface() {propertyValues {
       AlphaCurve {keys [<at 0 0>,<at 0.0599 6 RCIM_Cubic>,<at 0.1641 0 RCIM_Cubic>]}
       BlueCurve {keys [<at 0 0.12 RCIM_Cubic>]}
       GreenCurve {keys [<at 0 0.0836 RCIM_Cubic>]}
       RedCurve {keys [<at 0 0.0113 RCIM_Cubic>]}
      }})
    R4 TitaniumWhite:LinearColor=&ColorFromCurve(ColorCurve=DataInterface() {propertyValues {
       AlphaCurve {keys [<at 0 0>,<at 0.0599 6 RCIM_Cubic>,<at 0.1641 0 RCIM_Cubic>]}
       BlueCurve {keys [<at 0 0.08 RCIM_Cubic>]}
       GreenCurve {keys [<at 0 0.08 RCIM_Cubic>]}
       RedCurve {keys [<at 0 0.08 RCIM_Cubic>]}
      }})
    Saffron=&ColorFromCurve(ColorCurve=DataInterface() {propertyValues {
       AlphaCurve {keys [<at 0.6873 3 RCIM_Cubic>,<at 1 0 RCIM_Cubic>]}
       BlueCurve {keys [<at 0 0 RCIM_Cubic>]}
       GreenCurve {keys [<at 0 0.2047 RCIM_Cubic>]}
       RedCurve {keys [<at 0 0.22 RCIM_Cubic>]}
      }})
    SkyBlue=&ColorFromCurve(ColorCurve=DataInterface() {propertyValues {
       AlphaCurve {keys [<at 0.7017 2 RCIM_Cubic>,<at 1 0 RCIM_Cubic>]}
       BlueCurve {keys [<at 0 0.2 RCIM_Cubic>]}
       GreenCurve {keys [<at 0 0.1675 RCIM_Cubic>]}
       RedCurve {keys [<at 0 0 RCIM_Cubic>]}
      }})
    TitaniumWhite=&ColorFromCurve(ColorCurve=DataInterface() {propertyValues {
       AlphaCurve {keys [<at 0.7116 3 RCIM_Cubic>,<at 1 0 RCIM_Cubic>]}
       BlueCurve {keys [<at 0 0.3 RCIM_Cubic>]}
       GreenCurve {keys [<at 0 0.3 RCIM_Cubic>]}
       RedCurve {keys [<at 0 0.3 RCIM_Cubic>]}
      }})
    Use R1 Painted Overrides=true
    Use R2 Painted Overrides=true
    Use R3 Painted Overrides=true
    Use R4 Painted Overrides=true
   ) {module=VC_PaintedVariantsOverride001}
   &ScaleRibbonWidth(Ribbon Width Scale=&FloatFromCurve(FloatCurve=DataInterface() {propertyValues {Curve {keys [<at 0 1 RCIM_Cubic>,<at 1 0.6 RCIM_Cubic>]}}}))
   /SetVariables(
    Particles.Position0:NiagaraPosition=&AddVectorToPosition(Position=$Particles.Position, Vector=/VehicleCosmetics/FX/Niagara/Modules/VC_ConvertVectorToWorld(Vector:Vector3f=<0 0 -4>))
    Particles.Position1:NiagaraPosition=&AddVectorToPosition(Position=$Particles.Position, Vector=&VC_ConvertVectorToWorld(Vector:Vector3f=<0 0 -5>))
    Particles.RibbonWidth1=280
    Particles.RibbonWidth2=23
    Particles.RibbonWidth3=30
   ) {isSetParameters=true}
  ]
  rendererClasses [&NiagaraRibbonRendererProperties]
  renderers [
   {rendererClass=&NiagaraRibbonRendererProperties}
   {rendererClass=&NiagaraRibbonRendererProperties, rendererIndex=1}
   {rendererClass=&NiagaraRibbonRendererProperties, rendererIndex=2}
   {rendererClass=&NiagaraRibbonRendererProperties, rendererIndex=3}
  ]
 }
}
summary {
 systemName=NS_Boost_Techy
 emitters [
  {emitterName=Ribbon001, rendererClasses [&NiagaraRibbonRendererProperties]}
  {
   emitterName=NE_VC_Boost_Sprite_Parent002
   rendererClasses [&NiagaraSpriteRendererProperties, &NiagaraMeshRendererProperties]
  }
  {emitterName=NE_VC_Boost_Sprite_Parent003, rendererClasses [&NiagaraSpriteRendererProperties]}
 ]
 userVariables [
  User.bIsFrontendPreview {type=NiagaraBool}
  User.bIsPainted {type=NiagaraBool}
  User.ColorVariantID {type=NiagaraInt32}
  User.DTCarOffset {type=NiagaraPosition, description="", defaultValue {}}
  User.PaintedColor {type=LinearColor, description="", defaultValue {struct=LinearColor, value=<rgba 1 1 1 1>}}
  User.PaintedScalar {type=NiagaraFloat, defaultValue=60, description=""}
  User.PortCount {type=NiagaraInt32, defaultValue=1}
 ]
}
