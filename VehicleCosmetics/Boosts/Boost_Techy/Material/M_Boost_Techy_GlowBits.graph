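# Translucent unlit, two-sided Niagara sprite material for the glowing bits in the Techy vehicle boost effect.
# A TextureSwitch dynamic parameter, passed through Ceil/Floor branches, picks the blue or red channel of an outline-shapes texture; the inner blue-to-blue lerp is redundant.
# That mask tints squared particle color, which Hueshift actually desaturates, then Brightness and a glow scalar scale it into MF_FX_EmissiveAdjustments.
# Opacity is the heavily boosted, clamped mask times particle alpha.
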Material /VehicleCosmetics/Boosts/Boost_Techy/Material/M_Boost_Techy_GlowBits
graph {
 properties {
  bUsedWithNiagaraSprites=true
  blendMode=BLEND_Translucent
  pixelDepthOffsetMode=PDOM_Legacy
  shadingModel=MSM_Unlit
  twoSided=true
  usedShadingModels=Unlit
 }
 nodes [
  n1 = ParticleColor()
  n2 = DynamicParameter() {paramNames [Brightness, Hueshift, TextureSwitch, TextureSwitch]}
  n3 = TextureSample() {texture=/VehicleCosmetics/Boosts/Boost_Techy/Textures/T_VCVFX_Boost_Techy_ShapesOutline_RGB}
  n4 = Constant() {r=0.5}
  n5 = Ceil(None=@n2.TextureSwitch)
  n6 = Floor(None=@n2.TextureSwitch)
  n7 = LinearInterpolate(
   A=LinearInterpolate(A=@n3.B, B=@n3.B, Alpha=If(A=@n4, B=@n2.TextureSwitch, "A > B"=@n5, "A == B"=@n2.TextureSwitch, "A < B"=@n6))
   B=@n3.R
   Alpha=If(
    A=Constant() {r=0.5}
    B=@n2.TextureSwitch
    "A > B"=Ceil(None=@n2.TextureSwitch)
    "A == B"=@n2.TextureSwitch
    "A < B"=Floor(None=@n2.TextureSwitch)
   )
  )
  n8 = Multiply(A=Desaturation(None=Multiply(A=@n1.RGB, B=@n1.RGB), Fraction=@n2.Hueshift), B=@n7)
  MaterialFunctionCall(Emissive=Multiply(A=ScalarParameter() {defaultValue=3, parameterName=Emissive_PowerGlw}, B=Multiply(A=@n8, B=@n2.Brightness))) {materialFunction=&MF_FX_EmissiveAdjustments}
  n9 = Multiply(A=@n7, B=Constant() {r=60})
  Saturate(None=Multiply(A=Saturate(None=@n9), B=@n1.A))
 ]
}
