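# Translucent unlit material for the Techy vehicle boost effect on Niagara mesh particles.
# Layers panned trail and pan-time textures, blending multiplied and added versions through dynamic parameters, with particle color weighting the noise channels.
# The result, tinted by desaturated particle color, feeds emissive adjustments, while UV-edge masks and particle alpha shape opacity.
# Static switches select tile-distance UVs for the trail and enable optional distortion driven by the same pattern.
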Material /VehicleCosmetics/Boosts/Boost_Techy/Material/M__Boost_Techy_BoostMesh
graph {
 properties {
  bUsedWithNiagaraMeshParticles=true
  blendMode=BLEND_Translucent
  pixelDepthOffsetMode=PDOM_Legacy
  refractionMethod=RM_2DOffset
  shadingModel=MSM_Unlit
  usedShadingModels=Unlit
 }
 nodes [
  n1 = ParticleColor()
  n2 = Desaturation(None=@n1.RGB, Fraction=ScalarParameter() {parameterName=Saturation})
  n3 = TextureSample(UVs=Panner(Coordinate=TextureCoordinate() {uTiling=-0.25}, Time=Constant() {r=0.5}) {speedX=-0.1, speedY=-1}) {texture=/VehicleCosmetics/Boosts/Boost_Techy/Textures/T_VCVFX_Boost_Techy_PanTime02_RGB}
  n4 = TextureCoordinate()
  n5 = ComponentMask(None=@n4) {r=true}
  n6 = ConstantBiasScale(None=@n5) {bias=-0.01, scale=5}
  n7 = ConstantBiasScale(None=OneMinus(None=@n5)) {bias=-0.01, scale=5}
  n8 = ComponentMask(None=@n4) {g=true}
  n9 = ConstantBiasScale(None=@n8) {bias=-0.01, scale=5}
  n10 = ConstantBiasScale(None=OneMinus(None=@n8)) {bias=-0.01, scale=5}
  n11 = Multiply(A=Multiply(A=Saturate(None=@n7), B=@n5), B=Multiply(A=Saturate(None=@n9), B=Saturate(None=@n10)))
  n12 = Multiply(A=Multiply(A=Add(A=@n3.B, B=@n3.B), B=Saturate(None=@n6)), B=@n11)
  n13 = TextureCoordinate() {coordinateIndex=1, uTiling=0.75, vTiling=1.25}
  n14 = TextureCoordinate()
  n15 = Multiply(A=Time(), B=ScalarParameter() {defaultValue=2, parameterName=PanSpeed})
  n16 = TextureSample(UVs=Panner(Coordinate=StaticSwitchParameter(True=@n13, False=@n14) {parameterName=UseTileDistance}, Time=@n15) {speedX=0.05}) {texture=/VehicleCosmetics/Boosts/Boost_Techy/Textures/T_VCVFX_Boost_Techy_Trail03}
  n17 = Multiply(A=@n12, B=Multiply(A=Constant() {r=0.3}, B=@n16.B))
  n18 = Desaturation(None=@n16.RGB)
  n19 = DynamicParameter() {paramNames [Hardness,BlurAmount,AddMultiply,Brightness]}
  n20 = Multiply(A=@n19.BlurAmount, B=@n16.B)
  n21 = TextureSample(UVs=Panner(Coordinate=TextureCoordinate() {uTiling=-0.5}, Time=@n15) {speedX=-0.5}) {texture=&T_VCVFX_Boost_Techy_PanTime02_RGB}
  n22 = ParticleColor()
  n23 = Add(A=Add(A=Multiply(A=@n21.R, B=@n22.R), B=Multiply(A=@n21.G, B=@n22.G)), B=Multiply(A=@n21.B, B=@n22.B))
  n24 = LinearInterpolate(
   A=Multiply(A=Multiply(A=@n19.Hardness, B=@n18), B=Multiply(A=@n20, B=@n23))
   B=Multiply(A=@n19.Hardness, B=Add(A=@n18, B=@n23))
   Alpha=@n19.AddMultiply
  )
  n25 = Saturate(None=ConstantBiasScale(None=Add(A=@n24, B=@n17)) {bias=0.01, scale=5})
  MaterialFunctionCall(Emissive=Multiply(A=Multiply(A=@n2, B=@n25), B=@n19.Brightness)) {materialFunction=&MF_FX_EmissiveAdjustments}
  n26 = ScalarParameter() {defaultValue=20, parameterName=Power_EmissiveLines}
  n27 = Saturate(None=@n24)
  n28 = Multiply(A=Add(A=Multiply(A=@n26, B=@n27), B=@n11), B=@n12)
  n29 = ConstantBiasScale(None=@n11) {bias=-0.2, scale=0.15}
  Saturate(None=Add(A=Multiply(A=@n1.A, B=@n28), B=Saturate(None=@n29)))
  StaticSwitchParameter(True=Multiply(A=@n27, B=ScalarParameter() {defaultValue=10, parameterName=DistortionAmount}), False=Constant()) {parameterName=Distortion}
 ]
}
